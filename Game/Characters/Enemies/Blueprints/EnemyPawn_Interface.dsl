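; EnemyPawn_Interface is a Blueprint interface for enemy pawns, declaring the calls other code makes on them.
; Orphaned returns two outputs, defaulting to false and zero, whose meaning the file does not show.
; TriggerDeathFX takes no arguments and returns nothing, serving as a hook for each implementer to play its death effects.
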(fn Orphaned
  (return false 0))

(fn TriggerDeathFX)
