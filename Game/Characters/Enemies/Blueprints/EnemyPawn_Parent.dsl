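; Parent Blueprint for enemy pawns that supplies shared cosmetic effects for spawn-in, sleep, hit reacts, death and minibosses.
; Helpers swap materials with parameter copy and restore, build leader-pose duplicate meshes, and push parameters across body, head and auxiliary meshes.
; Awaken and building-hit effects lazily create FX mesh components cached in per-mesh maps.
; Death visuals choose particles by effects quality, and a begin-play timer recovers unfinished spawn-ins.
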(event EventBeginPlay
  (bind _fortpawncomponent_sleep (Variables|Default|GetFortPawnComponent_Sleep))
  (bind _elemental_layer_audio_loop (Variables|Sound|GetElemental_Layer_Audio_Loop))
  ;[1 1 1] Begin Play
  (Networking|CanExecuteCosmeticEvents
    (:True
      (Utilities|IsValid _fortpawncomponent_sleep
        (:Is%20Valid
          (Animation|BindEventtoOnBeginSleepEffects _fortpawncomponent_sleep (EventDispatchers|CreateEvent :!event OnBeginSleepEffects))
          (Animation|BindEventtoOnEndSleepEffects _fortpawncomponent_sleep (EventDispatchers|CreateEvent :!event OnEndSleepEffects)))
        (:Is%20Not%20Valid
          _))
      (Audio|Components|Audio|SetSound _elemental_layer_audio_loop (Variables|Sound|GetElemental_Audio_Layer_Sound))
      (Audio|Components|Audio|Play _elemental_layer_audio_loop 0.0)
      (bind _returnvalue (Utilities|Time|SetTimerbyFunctionName 0 "CharacterSpawnInSafetyCheck" 5.5))
      (Variables|Effects|SetCharacterSpawnInSafetyCheckHandle _returnvalue))))

(event EventTriggerDeathFX
  (CallFunction|BeginDeathFX))

(fn ApplyAwakenEffectToSkeletalMesh
  (params (Mesh :type Skeletal%20Mesh%20Component%20Object%20Reference))
  (bind _awakenfxskeletalmeshes (Variables|Effects|GetAwakenFXSkeletalMeshes))
  (bind (_value _returnvalue) (Utilities|Map|Find _awakenfxskeletalmeshes Mesh))
  (bind _fxskeletalmesh (LocalVariables|GetFXSkeletalMesh))
  (Utilities|IsValid Mesh
    (:Is%20Valid
      (if _returnvalue
        (bind _output_get (LocalVariables|SetFXSkeletalMesh _value))
        (Utilities|IsValid _output_get
          (:Is%20Valid
            _)
          (:Is%20Not%20Valid
            (bind _returnvalue_2 (AddComponent|NODE_AddFXSkeletonMeshComponent-6))
            (Utilities|Map|Add (Variables|Effects|GetAwakenFXSkeletalMeshes) Mesh _returnvalue_2)
            (bind _output_get_2 (LocalVariables|SetFXSkeletalMesh _returnvalue_2))
            (Class|FortFXSkeletonMeshComponent|SetSource _output_get_2 Mesh)
            (Variables|/Script/FortniteGame.FortFXSkeletonMeshComponent|SetAwakenDuration (Variables|Effects|GetAwakenDuration) _fxskeletalmesh)))
        (Class|FXSkeletonMeshComponent|ApplyAwakenEffect _fxskeletalmesh)
        (else
          _)))))

(fn ApplyAwakenEffectToStaticMesh
  (params (Mesh :type Static%20Mesh%20Component%20Object%20Reference))
  (bind _awakenfxstaticmeshes (Variables|Effects|GetAwakenFXStaticMeshes))
  (bind (_value _returnvalue) (Utilities|Map|Find _awakenfxstaticmeshes Mesh))
  (bind _fxstaticmesh (LocalVariables|GetFXStaticMesh))
  (Utilities|IsValid Mesh
    (:Is%20Valid
      (if _returnvalue
        (bind _output_get (LocalVariables|SetFXStaticMesh _value))
        (Utilities|IsValid _output_get
          (:Is%20Valid
            _)
          (:Is%20Not%20Valid
            (bind _returnvalue_2 (AddComponent|NODE_AddFXStaticMeshComponent-2))
            (Utilities|Map|Add (Variables|Effects|GetAwakenFXStaticMeshes) Mesh _returnvalue_2)
            (bind _output_get_2 (LocalVariables|SetFXStaticMesh _returnvalue_2))
            (Class|FortFXStaticMeshComponent|SetSource _output_get_2 Mesh)
            (Variables|/Script/FortniteGame.FortFXStaticMeshComponent|SetAwakenDuration (Variables|Effects|GetAwakenDuration) _fxstaticmesh)))
        (Class|FXStaticMeshComponent|ApplyAwakenEffect _fxstaticmesh)
        (else
          _)))))

(fn ApplyBuildingHitEffectToSkeletalMesh
  (params (Mesh :type Skeletal%20Mesh%20Component%20Object%20Reference))
  (bind _buildinghitfxskeletalmeshes (Variables|Effects|GetBuildingHitFXSkeletalMeshes))
  (bind (_value _returnvalue) (Utilities|Map|Find _buildinghitfxskeletalmeshes Mesh))
  (bind _fxskeletalmesh (LocalVariables|GetFXSkeletalMesh))
  (Utilities|IsValid Mesh
    (:Is%20Valid
      (if _returnvalue
        (bind _output_get (LocalVariables|SetFXSkeletalMesh _value))
        (Utilities|IsValid _output_get
          (:Is%20Valid
            _)
          (:Is%20Not%20Valid
            (bind _returnvalue_2 (AddComponent|NODE_AddFXSkeletonMeshComponent-2))
            (Utilities|Map|Add (Variables|Effects|GetBuildingHitFXSkeletalMeshes) Mesh _returnvalue_2)
            (bind _output_get_2 (LocalVariables|SetFXSkeletalMesh _returnvalue_2))
            (Class|FortFXSkeletonMeshComponent|SetSource _output_get_2 Mesh)
            (Variables|/Script/FortniteGame.FortFXSkeletonMeshComponent|SetBuildingHitDuration (Variables|Effects|GetBuildingHitDuration) _fxskeletalmesh)))
        (Class|FXSkeletonMeshComponent|ApplyBuildingHitEffect _fxskeletalmesh)
        (else
          _)))))

(fn ApplyBuildingHitEffectToStaticMesh
  (params (Mesh :type Static%20Mesh%20Component%20Object%20Reference))
  (bind _buildinghitfxstaticmeshes (Variables|Effects|GetBuildingHitFXStaticMeshes))
  (bind (_value _returnvalue) (Utilities|Map|Find _buildinghitfxstaticmeshes Mesh))
  (bind _fxstaticmesh (LocalVariables|GetFXStaticMesh))
  (Utilities|IsValid Mesh
    (:Is%20Valid
      (if _returnvalue
        (bind _output_get (LocalVariables|SetFXStaticMesh _value))
        (Utilities|IsValid _output_get
          (:Is%20Valid
            _)
          (:Is%20Not%20Valid
            (bind _returnvalue_2 (AddComponent|NODE_AddFXStaticMeshComponent-0))
            (Utilities|Map|Add (Variables|Effects|GetBuildingHitFXStaticMeshes) Mesh _returnvalue_2)
            (bind _output_get_2 (LocalVariables|SetFXStaticMesh _returnvalue_2))
            (Class|FortFXStaticMeshComponent|SetSource _output_get_2 Mesh)
            (Variables|/Script/FortniteGame.FortFXStaticMeshComponent|SetBuildingHitDuration (Variables|Effects|GetBuildingHitDuration) _fxstaticmesh)))
        (Class|FXStaticMeshComponent|ApplyBuildingHitEffect _fxstaticmesh)
        (else
          _)))))

(fn CharacterSpawnInSafetyCheck
  (if (not (Variables|Effects|GetSpawnInTimelineCompletedSuccessfully))
    (Components|Timeline|Stop (Variables|EnemyPawn_Parent|GetEnemySpawnInTL))
    (CallFunction|SpawninSafetyCheck)))

(fn CreateDuplicateCharacterElements
  (params (Material%20To%20Apply :type Material%20Interface%20Object%20Reference))
  (bind _auxillary_objects (Variables|ElementalFXs|GetAuxillary%20Objects))
  (bind _duplicate_auxillary_objects (Variables|ElementalFXs|GetDuplicate%20Auxillary%20Objects))
  (bind _mesh (Variables|Character|GetMesh))
  (bind _duplicatecharactermesh (Variables|Effects|GetDuplicateCharacterMesh))
  (bind _material_to_apply (Variables|GetMaterial%20To%20Apply))
  (bind _duplicatecharactermesh_2 (Variables|Effects|GetDuplicateCharacterMesh))
  (bind _material_to_apply_2 (Variables|GetMaterial%20To%20Apply))
  (Networking|CanExecuteCosmeticEvents
    (:True
      (Variables|Effects|GetDuplicateCharacterMesh
        (:then
          _)
        (:else
          (bind _returnvalue (AddComponent|NODE_AddSkeletalMeshComponent-1))
          (bind _output_get (Variables|Effects|SetDuplicateCharacterMesh _returnvalue))))
      (Components|SkinnedMesh|SetSkinnedAssetandUpdate _output_get (Variables|/Script/Engine.SkinnedMeshComponent|GetSkeletalMesh _mesh) true)
      (Components|SkinnedMesh|SetLeaderPoseComponent _duplicatecharactermesh _mesh false false)
      (Transformation|AttachComponentToComponent _duplicatecharactermesh _mesh "None" "SnapToTarget" "SnapToTarget" "SnapToTarget" true)
      (Utilities|IsValid Material%20To%20Apply
        (:Is%20Valid
          (for (_ _array_index) (Rendering|Material|GetMaterials _duplicatecharactermesh_2)
            (Rendering|Material|CreateDynamicMaterialInstance _duplicatecharactermesh_2 _array_index Material%20To%20Apply "None" :!class /Script/Engine.PrimitiveComponent))
          (if (> (Utilities|Array|Length _auxillary_objects) 0)
            (for (_array_element _array_index_2) _auxillary_objects
              (if (.bSuccess (Utilities|Casting|CastToStaticMeshComponent _array_element))
                (bind _returnvalue_2 (AddComponent|NODE_AddStaticMeshComponent-4))
                (Utilities|Array|SetArrayElem _duplicate_auxillary_objects _array_index_2 _returnvalue_2 true)
                (Components|StaticMesh|SetStaticMesh _returnvalue_2 (Variables|/Script/Engine.StaticMeshComponent|GetStaticMesh (Utilities|Casting|CastToStaticMeshComponent _array_element)))
                (Collision|SetCollisionEnabled _returnvalue_2 "NoCollision")
                (Transformation|AttachComponentToComponent _returnvalue_2 (Utilities|Casting|CastToStaticMeshComponent _array_element) "None" "SnapToTarget" "SnapToTarget" "SnapToTarget" true)
                (Utilities|IsValid _material_to_apply_2
                  (:Is%20Valid
                    (for (_ _array_index_3) (Rendering|Material|GetMaterials _returnvalue_2)
                      (Rendering|Material|CreateDynamicMaterialInstance _returnvalue_2 _array_index_3 _material_to_apply_2 "None" :!class /Script/Engine.PrimitiveComponent))))
                (elif (.bSuccess (Utilities|Casting|CastToSkeletalMeshComponent _array_element))
                  (bind _returnvalue_3 (AddComponent|NODE_AddSkeletalMeshComponent-3))
                  (Utilities|Array|SetArrayElem _duplicate_auxillary_objects _array_index_2 _returnvalue_3 true)
                  (Components|SkinnedMesh|SetSkinnedAssetandUpdate _returnvalue_3 (Variables|/Script/Engine.SkinnedMeshComponent|GetSkeletalMesh (Utilities|Casting|CastToSkeletalMeshComponent _array_element)) true)
                  (Collision|SetCollisionEnabled _returnvalue_3 "NoCollision")
                  (Components|SkinnedMesh|SetLeaderPoseComponent _returnvalue_3 (Utilities|Casting|CastToSkeletalMeshComponent _array_element) false false)
                  (Transformation|AttachComponentToComponent _returnvalue_3 (Utilities|Casting|CastToSkeletalMeshComponent _array_element) "None" "SnapToTarget" "SnapToTarget" "SnapToTarget" true)
                  (Utilities|IsValid _material_to_apply
                    (:Is%20Valid
                      (for (_ _array_index_4) (Rendering|Material|GetMaterials _returnvalue_3)
                        (Rendering|Material|CreateDynamicMaterialInstance _returnvalue_3 _array_index_4 _material_to_apply "None" :!class /Script/Engine.PrimitiveComponent)))))))))))))

(fn DeleteDuplicateCharacterElements
  (bind _duplicate_auxillary_objects (Variables|ElementalFXs|GetDuplicate%20Auxillary%20Objects))
  (Networking|CanExecuteCosmeticEvents
    (:True
      (bind _duplicatecharactermesh (Variables|Effects|GetDuplicateCharacterMesh)
        (:then
          (Components|DestroyComponent _duplicatecharactermesh)
          (if (> (Utilities|Array|Length _duplicate_auxillary_objects) 0)
            (for _array_element _duplicate_auxillary_objects
              (if (.bSuccess (Utilities|Casting|CastToMeshComponent _array_element))
                (Components|DestroyComponent (Utilities|Casting|CastToMeshComponent _array_element))))))
        (:else)))))

(fn DeregisterEffectSkeletalMesh
  (params (Mesh :type Skeletal%20Mesh%20Component%20Object%20Reference))
  (bind _awakenfxskeletalmeshes (Variables|Effects|GetAwakenFXSkeletalMeshes))
  (bind (_value _returnvalue) (Utilities|Map|Find _awakenfxskeletalmeshes Mesh))
  (bind _buildinghitfxskeletalmeshes (Variables|Effects|GetBuildingHitFXSkeletalMeshes))
  (bind (_value_2 _returnvalue_2) (Utilities|Map|Find _buildinghitfxskeletalmeshes Mesh))
  (Utilities|IsValid Mesh
    (:Is%20Valid
      (if _returnvalue
        (Utilities|IsValid _value
          (:Is%20Valid
            (Components|DestroyComponent _value))
          (:Is%20Not%20Valid
            _))
        (Utilities|Map|Remove (Variables|Effects|GetAwakenFXSkeletalMeshes) Mesh)
        (if _returnvalue_2
          (Utilities|IsValid _value_2
            (:Is%20Valid
              (Components|DestroyComponent _value_2))
            (:Is%20Not%20Valid
              _))
          (Utilities|Map|Remove (Variables|Effects|GetBuildingHitFXSkeletalMeshes) Mesh)
          (Utilities|Array|RemoveItem (Variables|Effects|GetRegisteredSkeletalMeshesForEffects) Mesh))))))

(fn DeregisterEffectStaticMesh
  (params (Mesh :type Static%20Mesh%20Component%20Object%20Reference))
  (bind _awakenfxstaticmeshes (Variables|Effects|GetAwakenFXStaticMeshes))
  (bind (_value _returnvalue) (Utilities|Map|Find _awakenfxstaticmeshes Mesh))
  (bind _buildinghitfxstaticmeshes (Variables|Effects|GetBuildingHitFXStaticMeshes))
  (bind (_value_2 _returnvalue_2) (Utilities|Map|Find _buildinghitfxstaticmeshes Mesh))
  (Utilities|IsValid Mesh
    (:Is%20Valid
      (if _returnvalue
        (Utilities|IsValid _value
          (:Is%20Valid
            (Components|DestroyComponent _value))
          (:Is%20Not%20Valid
            _))
        (Utilities|Map|Remove (Variables|Effects|GetAwakenFXStaticMeshes) Mesh)
        (if _returnvalue_2
          (Utilities|IsValid _value_2
            (:Is%20Valid
              (Components|DestroyComponent _value_2))
            (:Is%20Not%20Valid
              _))
          (Utilities|Map|Remove (Variables|Effects|GetBuildingHitFXStaticMeshes) Mesh)
          (Utilities|Array|RemoveItem (Variables|Effects|GetRegisteredStaticMeshesForEffects) Mesh))))))

(fn DynamicMaterialArraySetup
  (params (Array%20to%20Check :type Array%20of%20Material%20Interface%20Object%20References) (Array%20to%20Store :type Array%20of%20Material%20Instance%20Dynamic%20Object%20References) (New%20Material :type Material%20Interface%20Object%20Reference) (Mesh%20to%20Assign%20New%20Material :type Primitive%20Component%20Object%20Reference))
  (for (_array_element _array_index) Array%20to%20Check
    (Utilities|IsValid _array_element
      (:Is%20Valid
        (bind _asmaterial_instance_dynamic (Utilities|Casting|CastToMaterialInstanceDynamic _array_element)
          (:then
            (if (not (Utilities|Equal(Object) New%20Material (Variables|/Script/Engine.MaterialInstance|GetParent _asmaterial_instance_dynamic)))
              (Utilities|Array|Add Array%20to%20Store _asmaterial_instance_dynamic)
              (bind _returnvalue (Rendering|Material|CreateDynamicMaterialInstance New%20Material "None" "None" :!class /Script/Engine.KismetMaterialLibrary))
              (Rendering|Material|CopyMaterialInstanceParameters _returnvalue _asmaterial_instance_dynamic true)
              (Rendering|Material|SetMaterial Mesh%20to%20Assign%20New%20Material _array_index _returnvalue)))
          (:CastFailed
            (bind _returnvalue_2 (Rendering|Material|CreateDynamicMaterialInstance _array_element "None" "None" :!class /Script/Engine.KismetMaterialLibrary))
            (Utilities|Array|Add Array%20to%20Store _returnvalue_2)
            (bind _returnvalue_3 (Rendering|Material|CreateDynamicMaterialInstance New%20Material "None" "None" :!class /Script/Engine.KismetMaterialLibrary))
            (Rendering|Material|CopyMaterialInstanceParameters _returnvalue_3 _returnvalue_2 true)
            (Rendering|Material|SetMaterial Mesh%20to%20Assign%20New%20Material _array_index _returnvalue_3))))))
  (return))

(fn EnemyDeathVisuals
  (bind (_origin _boxextent _sphereradius) (Collision|GetComponentBounds (Variables|Character|GetMesh)))
  (Networking|CanExecuteCosmeticEvents
    (:True
      (if (Variables|Health|GetbIsDying)
        ;[1 1 1] Set material parameters
        (bind _deathhitsocket (Variables|Death|GetDeathHitSocket)
          (:then
            (CallFunction|SetVectorParameteronAllCharacterMIDs "HitLocation" (CallFunction|TransformHitLocationWorldSpacetoPreskinnedLocalSpace))
            (Rendering|Material|SetScalarParameterValueonMaterials (Variables|Character|GetMesh) "Object Radius" (Utilities|DeathSetSphericalRadius (Variables|Character|GetMesh) _deathhitsocket)))
          (:else
            _))
        (bind (_vfx_quality_int _quality_check_bool) (Utilities|EffectsQualityCheck 2)
          (:Cinematic
            _)
          (:Epic
            _)
          (:High
            _)
          (:Medium
            ;[1 1 1] Low Quality Particles
            (bind _death_effects (Variables|Effects%28Death%29|GetDeath_Effects)
              (:then
                (Components|DestroyComponent _death_effects))
              (:else
                _))
            (bind _pscomponentreference (CallFunction|SpawnParticleSystemonCharacterMesh (Variables|Effects%28Death%29|GetLQ%20Death%20Particle%20System) 0 (Variables|Effects%28Death%29|GetLQ%20PS%20Socket%20Attach) "0, 0, 0" "0, 0, 0" (Utilities|Array|MakeArray (Utilities|Struct|MakeParticleSysParam "SourceCylinderHeight" "PSPT_Scalar" (Variables|Effects%28Death%29|GetLQ_DeathFX_CylinderHeight)) (Utilities|Struct|MakeParticleSysParam "SourceCylinderRadius" "PSPT_Scalar" (Variables|Effects%28Death%29|GetLQ_DeathFX_CylinderRadius)) (Utilities|Struct|MakeParticleSysParam "SpawnRateScale" "PSPT_Scalar" (Variables|Effects%28Death%29|GetLQ_DeathFX_SpawnRateScale))) true true))
            (Variables|Effects%28Death%29|SetDeath_Effects _pscomponentreference)
            ;[0 0 0] Additive Particle sys
            (if (Variables|Effects%28Death%29|GetDeathParticleSystem_Additive_On?)
              (bind _returnvalue (Effects|SpawnEmitteratLocation (Variables|Effects%28Death%29|GetDeathParticleSystem_Additive) (Transformation|GetWorldLocation (Variables|Effects%28Death%29|GetDeath_Effects))))
              (bind _output_get (Variables|Effects%28Death%29|SetAdditive_Death_Effects _returnvalue))
              (Variables|/Script/CascadeRuntime.ParticleSystemComponent|SetInstanceParameters (Utilities|Array|MakeArray (Utilities|Struct|MakeParticleSysParam "BP_Radius" "PSPT_Scalar" (* _sphereradius 0.5)) (Utilities|Struct|MakeParticleSysParam "BP_Hight" "PSPT_Scalar" (* (.z _boxextent) 0.5)) (Utilities|Struct|MakeParticleSysParam "BP_SpawnRate" "PSPT_Scalar" (Variables|Effects%28Death%29|GetDeathParticleSystem_Additive_Rate))) _output_get))
            (return _quality_check_bool))
          (:Low
            _))
        ;[1 1 1] High Quality Particles
        (bind _death_effects_2 (Variables|Effects%28Death%29|GetDeath_Effects)
          (:then
            (Components|DestroyComponent _death_effects_2))
          (:else
            _))
        (bind _pscomponentreference_2 (CallFunction|SpawnParticleSystemonCharacterMesh (Variables|Effects%28Death%29|GetHQ%20Death%20Particle%20System) 0 "FX_Chest" "0, 0, 0" "0, 0, 0" (Utilities|Array|MakeArray (Utilities|Struct|MakeParticleSysParam "EnemyPawn" "PSPT_Actor" 0.0 0.0 "0.000000,0.000000,0.000000" "0.000000,0.000000,0.000000" "(B=0,G=0,R=0,A=0)" self) (Utilities|Struct|MakeParticleSysParam "SpawnRateScale" "PSPT_Scalar" (Variables|Effects%28Death%29|GetDeathFX_SpawnRateScale))) true true))
        (Variables|Effects%28Death%29|SetDeath_Effects _pscomponentreference_2)))))

(fn Orphaned
  (return false 0))

(fn OverrideMaterialAndCopyParametersOnCharacterElements
  (params (New%20Material%20To%20Apply :type Material%20Interface%20Object%20Reference))
  (bind _auxillary_objects (Variables|ElementalFXs|GetAuxillary%20Objects))
  (Utilities|Array|Clear (Variables|Materials|GetPrevious%20MID))
  (Utilities|Array|Clear (Variables|Materials|GetPrevious%20MID_AuxObjs))
  (Networking|CanExecuteCosmeticEvents
    (:True
      (Utilities|IsValid New%20Material%20To%20Apply
        (:Is%20Valid
          (CallFunction|DynamicMaterialArraySetup (Rendering|Material|GetMaterials (Variables|Character|GetMesh)) (Variables|Materials|GetPrevious%20MID) New%20Material%20To%20Apply (Variables|Character|GetMesh))
          (if (> (Utilities|Array|Length _auxillary_objects) 0)
            (for _array_element _auxillary_objects
              (Utilities|IsValid _array_element
                (:Is%20Valid
                  (bind _asmesh_component (Utilities|Casting|CastToMeshComponent _array_element))
                  (CallFunction|DynamicMaterialArraySetup (Rendering|Material|GetMaterials _asmesh_component) (Variables|Materials|GetPrevious%20MID_AuxObjs) (Variables|GetNew%20Material%20To%20Apply) _asmesh_component))))))))))

(fn OverridePhysicalMaterialOnCharacterMesh
  (params (Physical%20Material%20Override :type Physical%20Material%20Object%20Reference))
  (bind _mesh (Variables|Character|GetMesh))
  (Networking|CanExecuteCosmeticEvents
    (:True
      (Utilities|Array|Clear (Variables|Materials|GetOriginal%20Phys%20Material))
      (Utilities|IsValid Physical%20Material%20Override
        (:Is%20Valid
          (for (_ _array_index) (Rendering|Material|GetMaterials _mesh)
            (bind _returnvalue (Fortnite|GetDynamicMaterialInstance _mesh _array_index))
            (Utilities|Array|SetArrayElem (Variables|Materials|GetOriginal%20Phys%20Material) _array_index (Physics|Material|GetPhysicalMaterial _returnvalue) true)
            (Variables|/Script/Engine.MaterialInstance|SetPhysMaterial Physical%20Material%20Override _returnvalue)))))))

(fn PickColorFromAnArrayOfColors
  (params (ArrayOfColors :type Array%20of%20Linear%20Color%20Structures))
  (return (Utilities|Array|Get(acopy) ArrayOfColors (Math|Integer|%25%28Integer%29 (Variables|Animation|GetPawnUniqueID) (Utilities|Array|LastIndex ArrayOfColors)))))

(fn PlayAdditiveHitReacts
  (params (Hit%20Direction :type Vector) (Anim%20Montage :type Anim%20Montage%20Object%20Reference))
  (bind _nameresult (FortPawn|SetFNameBasedonHitDirection Hit%20Direction "Forward" "Back" "Left" "Right"))
  (Animation|PlayLocalAnimMontage Anim%20Montage 1.0 _nameresult))

(fn RegisterEffectSkeletalMesh
  (params (Mesh :type Skeletal%20Mesh%20Component%20Object%20Reference))
  (Utilities|IsValid Mesh
    (:Is%20Valid
      (Utilities|Array|Add (Variables|Effects|GetRegisteredSkeletalMeshesForEffects) Mesh))
    (:Is%20Not%20Valid
      (Development|PrintString "Attempting to register invalid skeletal mesh for husk building hit FX"))))

(fn RegisterEffectStaticMesh
  (params (Mesh :type Static%20Mesh%20Component%20Object%20Reference))
  (Utilities|IsValid Mesh
    (:Is%20Valid
      (Utilities|Array|Add (Variables|Effects|GetRegisteredStaticMeshesForEffects) Mesh))
    (:Is%20Not%20Valid
      (Development|PrintString "Attempting to register invalid static mesh for husk building hit FX"))))

(fn RestorePreviousMaterialOnCharacterElements
  (params (Delay%20in%20Seconds :type Float%20%28double-precision%29))
  (bind _auxillary_objects (Variables|ElementalFXs|GetAuxillary%20Objects))
  (bind _previous_mid_auxobjs (Variables|Materials|GetPrevious%20MID_AuxObjs))
  (Networking|CanExecuteCosmeticEvents
    (:True
      (if (> Delay%20in%20Seconds 0.0)
        (CallFunction|RestorePreviousMaterialDelay Delay%20in%20Seconds)
        (else
          (for (_array_element _array_index) (Variables|Materials|GetPrevious%20MID)
            (Utilities|IsValid _array_element
              (:Is%20Valid
                (bind _returnvalue (Fortnite|GetDynamicMaterialInstance (Variables|Character|GetMesh) _array_index))
                (if (not (Utilities|Equal(Object) _returnvalue _array_element))
                  (Rendering|Material|SetMaterial (Variables|Character|GetMesh) _array_index _array_element)))))
          (Utilities|Array|Clear (Variables|Materials|GetPrevious%20MID))
          (if (> (Utilities|Array|Length _auxillary_objects) 0)
            (for _array_element_2 _auxillary_objects
              (Utilities|IsValid _array_element_2
                (:Is%20Valid
                  (bind _asmesh_component (Utilities|Casting|CastToMeshComponent _array_element_2))
                  (for (_array_element_3 _array_index_2) (Rendering|Material|GetMaterials _asmesh_component)
                    (Utilities|IsValid _array_element_3
                      (:Is%20Valid
                        (if (not (Utilities|Equal(Object) _array_element_3 (Utilities|Array|Get(acopy) _previous_mid_auxobjs _array_index_2)))
                          (Rendering|Material|SetMaterial _asmesh_component _array_index_2 (Utilities|Array|Get(acopy) _previous_mid_auxobjs _array_index_2)))))))))
            (Utilities|Array|Clear (Variables|Materials|GetPrevious%20MID_AuxObjs))))))))

(fn RestorePreviousPhysMaterialOnCharacterMesh
  (params (Delay%20in%20Seconds :type Float%20%28double-precision%29))
  (Networking|CanExecuteCosmeticEvents
    (:True
      (if (> Delay%20in%20Seconds 0.0)
        (CallFunction|RestorePreviousMaterialDelay Delay%20in%20Seconds)
        (else
          (for (_array_element _array_index) (Variables|Materials|GetPrevious%20MID)
            (Utilities|IsValid _array_element
              (:Is%20Valid
                (bind _returnvalue (Fortnite|GetDynamicMaterialInstance (Variables|Character|GetMesh) _array_index))
                (if (Utilities|Equal(Object) _returnvalue _array_element)
                  (Variables|/Script/Engine.MaterialInstance|SetPhysMaterial (Utilities|Array|Get(acopy) (Variables|Materials|GetOriginal%20Phys%20Material) _array_index) _array_element))))))))))

(fn SetActiveElementalParticles
  (params (Active :type Boolean) (Reset :type Boolean))
  (for _array_element (Variables|ElementalFXs|GetElemental%20FXs)
    (Utilities|IsValid _array_element
      (:Is%20Valid
        (Components|Activation|SetActive _array_element Active Reset)))))

(fn SetDuplicateCharacterMeshHiddenInGame
  (params (Hidden :type Boolean))
  (bind _duplicate_auxillary_objects (Variables|ElementalFXs|GetDuplicate%20Auxillary%20Objects))
  (Networking|CanExecuteCosmeticEvents
    (:True
      (bind _duplicatecharactermesh (Variables|Effects|GetDuplicateCharacterMesh)
        (:then
          (Development|SetHiddeninGame _duplicatecharactermesh Hidden false)
          (if (> (Utilities|Array|Length _duplicate_auxillary_objects) 0)
            (for _array_element _duplicate_auxillary_objects
              (Utilities|IsValid _array_element
                (:Is%20Valid
                  (bind _asmesh_component (Utilities|Casting|CastToMeshComponent _array_element))
                  (Development|SetHiddeninGame _asmesh_component (Variables|GetHidden) false))))))
        (:else)))))

(fn SetElementalParticlesHiddenInGame
  (params (Hidden :type Boolean))
  (for _array_element (Variables|ElementalFXs|GetElemental%20FXs)
    (Utilities|IsValid _array_element
      (:Is%20Valid
        (Development|SetHiddeninGame _array_element Hidden false)))))

(fn SetMinibossLightHiddenInGame
  (params (Hidden :type Boolean))
  (bind _minibosslight (Variables|MiniBoss|GetMiniBossLight)
    (:then
      (Development|SetHiddeninGame _minibosslight Hidden false))
    (:else)))

(fn SetScalarParameterOnAllCharacterMIDs
  (params (Parameter%20Name :type Name) (Scalar%20Value :type Float%20%28double-precision%29) (AlsoModifyOriginalMIDs :type Boolean))
  (bind _auxillary_objects (Variables|ElementalFXs|GetAuxillary%20Objects))
  (bind _duplicate_auxillary_objects (Variables|ElementalFXs|GetDuplicate%20Auxillary%20Objects))
  (Utilities|FlowControl|Sequence
    (:then_0
      (bind _mesh (Variables|Character|GetMesh)
        (:then
          (Rendering|Material|SetScalarParameterValueonMaterials _mesh (Variables|GetParameter%20Name) (Variables|GetScalar%20Value)))
        (:else
          _))
      (bind _duplicatecharactermesh (Variables|Effects|GetDuplicateCharacterMesh)
        (:then
          (Rendering|Material|SetScalarParameterValueonMaterials _duplicatecharactermesh (Variables|GetParameter%20Name) (Variables|GetScalar%20Value)))
        (:else
          _))
      (bind _head_mid (Variables|Default|GetHead%20MID)
        (:then
          (Rendering|Material|SetScalarParameterValue _head_mid (Variables|GetParameter%20Name) (Variables|GetScalar%20Value) :!class /Script/Engine.MaterialInstanceDynamic))
        (:else)))
    (:then_1
      (if (Utilities|IsArrayNonEmpty _auxillary_objects)
        (for _array_element _auxillary_objects
          (Utilities|IsValid _array_element
            (:Is%20Valid
              (bind _asmesh_component (Utilities|Casting|CastToMeshComponent _array_element))
              (Rendering|Material|SetScalarParameterValueonMaterials _asmesh_component (Variables|GetParameter%20Name) (Variables|GetScalar%20Value)))))))
    (:then_2
      (if (Utilities|IsArrayNonEmpty _duplicate_auxillary_objects)
        (for _array_element_2 _duplicate_auxillary_objects
          (Utilities|IsValid _array_element_2
            (:Is%20Valid
              (bind _asmesh_component_2 (Utilities|Casting|CastToMeshComponent _array_element_2))
              (Rendering|Material|SetScalarParameterValueonMaterials _asmesh_component_2 (Variables|GetParameter%20Name) (Variables|GetScalar%20Value)))))))
    (:then_3
      ;[1 1 1] Optionally keep the original MIDs in sync with the override MIDs
      (if AlsoModifyOriginalMIDs
        (Utilities|FlowControl|Sequence
          (:then_0
            (for _array_element_3 (Variables|Materials|GetPrevious%20MID)
              (Rendering|Material|SetScalarParameterValue _array_element_3 (Variables|GetParameter%20Name) (Variables|GetScalar%20Value) :!class /Script/Engine.MaterialInstanceDynamic)))
          (:then_1
            (for _array_element_4 (Variables|Materials|GetPrevious%20MID_AuxObjs)
              (Rendering|Material|SetScalarParameterValue _array_element_4 (Variables|GetParameter%20Name) (Variables|GetScalar%20Value) :!class /Script/Engine.MaterialInstanceDynamic))))))))

(fn SetScalarParameterOnAllCharacterMIDsByMap
  (params (Map :type Map%20of%20Names%20to%20Float%20%28double-precision%29s))
  (bind _keys (Utilities|Map|Keys Map))
  (bind _values (Utilities|Map|Values Map))
  (for (_array_element _array_index) _values
    (CallFunction|SetScalarParameteronAllCharacterMIDs (Utilities|Array|Get(acopy) _keys _array_index) _array_element))
  (return))

(fn SetVectorParameterOnAllCharacterMIDs
  (params (Parameter%20Name :type Name) (Vector%20Value :type Vector) (Propagate%20to%20Auxiliary%20Meshes :type Boolean :default true))
  (bind _auxillary_objects (Variables|ElementalFXs|GetAuxillary%20Objects))
  (bind _duplicate_auxillary_objects (Variables|ElementalFXs|GetDuplicate%20Auxillary%20Objects))
  (bind _mesh (Variables|Character|GetMesh)
    (:then
      (Rendering|Material|SetVectorParameterValueonMaterials _mesh Parameter%20Name Vector%20Value))
    (:else
      _))
  (bind _duplicatecharactermesh (Variables|Effects|GetDuplicateCharacterMesh)
    (:then
      (Rendering|Material|SetVectorParameterValueonMaterials _duplicatecharactermesh Parameter%20Name Vector%20Value))
    (:else
      _))
  (bind _head_mid (Variables|Default|GetHead%20MID)
    (:then
      (Rendering|Material|SetVectorParameterValue _head_mid Parameter%20Name (Math|Conversions|ToLinearColor(Vector) Vector%20Value) :!class /Script/Engine.MaterialInstanceDynamic))
    (:else
      _))
  (if (and (> (Utilities|Array|Length _auxillary_objects) 0) (Variables|GetPropagate%20to%20Auxiliary%20Meshes))
    (for _array_element _auxillary_objects
      (Utilities|IsValid _array_element
        (:Is%20Valid
          (bind _asmesh_component (Utilities|Casting|CastToMeshComponent _array_element))
          (Rendering|Material|SetVectorParameterValueonMaterials _asmesh_component (Variables|GetParameter%20Name) (Variables|GetVector%20Value)))))
    (if (and (> (Utilities|Array|Length _duplicate_auxillary_objects) 0) (Variables|GetPropagate%20to%20Auxiliary%20Meshes))
      (for _array_element_2 _duplicate_auxillary_objects
        (Utilities|IsValid _array_element_2
          (:Is%20Valid
            (bind _asmesh_component_2 (Utilities|Casting|CastToMeshComponent _array_element_2))
            (Rendering|Material|SetVectorParameterValueonMaterials _asmesh_component_2 (Variables|GetParameter%20Name) (Variables|GetVector%20Value)))))))
  (return))

(fn SetVectorParameterOnAllCharacterMIDsByMap
  (params (Map :type Map%20of%20Names%20to%20Linear%20Color%20Structures) (Propagate%20to%20Auxiliary%20Meshes :type Boolean :default true))
  (bind _values (Utilities|Map|Values Map))
  (bind _keys (Utilities|Map|Keys Map))
  (for (_array_element _array_index) _keys
    (CallFunction|SetVectorParameteronAllCharacterMIDs _array_element (Math|Conversions|ToVector(LinearColor) (Utilities|Array|Get(aref) _values _array_index)) Propagate%20to%20Auxiliary%20Meshes))
  (return))

(fn SpawnMinibossLight
  (bind (_returnvalue_x _returnvalue_y _returnvalue_z) (Math|Vector|BreakVector (- (Transformation|GetSocketLocation (Variables|Character|GetMesh) "Death_FX") (Transformation|GetSocketLocation (Variables|Character|GetMesh) "Root"))))
  (bind _returnvalue (AddComponent|NODE_AddPointLightComponent-0 :RelativeTransform (Math|Transform|MakeTransform (Math|Vector|MakeVector 0.0 0.0 _returnvalue_z) "0, 0, 0" "1.000000,1.000000,1.000000")))
  (Variables|MiniBoss|SetMiniBossLight _returnvalue))

(fn SpawnParticleSystemOnCharacterMesh
  (params (ParticleSystemTemplate :type Particle%20System%20Object%20Reference) (ParticleSystemComponentReferenceVar :type Cascade%20Particle%20System%20Component%20Object%20Reference) (AttachPointName :type Name) (Location :type Vector) (Rotation :type Rotator) (InstanceParameters :type Array%20of%20Particle%20Sys%20Param%20Structures) (AutoActivate :type Boolean :default true) (AutoDestroy :type Boolean) (AbsoluteLocation :type Boolean) (AbsoluteRotation :type Boolean) (AbsoluteScale :type Boolean))
  (Networking|CanExecuteCosmeticEvents
    (:True
      (Utilities|IsValid ParticleSystemTemplate
        (:Is%20Valid
          (Utilities|IsValid ParticleSystemComponentReferenceVar
            (:Is%20Valid
              (Components|Activation|Deactivate ParticleSystemComponentReferenceVar))
            (:Is%20Not%20Valid
              _))
          (bind _returnvalue (Effects|SpawnEmitterAttached ParticleSystemTemplate (Variables|Character|GetMesh) AttachPointName Location Rotation "1.000000,1.000000,1.000000" "KeepRelativeOffset" AutoDestroy))
          (Variables|/Script/CascadeRuntime.ParticleSystemComponent|SetInstanceParameters InstanceParameters _returnvalue)
          (Components|Activation|SetActive _returnvalue AutoActivate false)
          (Transformation|SetAbsolute _returnvalue AbsoluteLocation AbsoluteRotation AbsoluteScale)
          (return _returnvalue))))))

(fn StopDeathFX
  (CallFunction|CEStopDeathFX))

(fn TransformHitLocationWorldSpaceToPreskinnedLocalSpace
  (return (+ (Math|Transform|InverseTransformLocation (Transformation|GetWorldTransform (Variables|Character|GetMesh)) (Transformation|GetSocketLocation (Variables|Character|GetMesh) (Variables|/Script/Engine.SkeletalMeshSocket|GetBoneName (Variables|Death|GetDeathHitSocket)))) (Variables|Effects%28Death%29|GetBoneRefPosePosition))))

(fn ConstructionScript
  (Networking|CanExecuteCosmeticEvents
    (:True
      (AI|SetWeaponCollisionComponent (Variables|Default|GetWeaponCapsuleCollision))
      (bind _returnvalue (Rendering|Material|GetScalarParameterValue "/Game/Packages/Fortress_SharedMaterials/GlobalMaterialParameters/FortniteMaterialParameters.FortniteMaterialParameters" "Enemy Indirect Shadow Min Occlusion" :!class /Script/Engine.KismetMaterialLibrary))
      (Lighting|SetCapsuleIndirectShadowMinVisibility (Variables|Character|GetMesh) _returnvalue)
      (CallFunction|RegisterEffectSkeletalMesh (Variables|Character|GetMesh)))))
