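; Enemy FX skeleton mesh component that shows timed material effects for awakening (seen through walls, per the material name) and hitting a building.
; Both wrappers pass their material and timing variables to ApplyEffect, which creates a dynamic instance for each source mesh slot.
; Game-time timestamps and transition lengths let the material drive its own fade in and out.
; Valid normal and diffuse textures are copied from the source mesh's materials.
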(fn ApplyAwakenEffect
  (CallFunction|ApplyEffect "/Game/Characters/Enemies/Husk/Materials/M_EnemyAwokenThroughWalls_01.M_EnemyAwokenThroughWalls_01" (Variables|Effects|GetAwakenFadeInTime) (Variables|Effects|GetAwakenDuration) (Variables|Effects|GetAwakenFadeOutTime)))

(fn ApplyBuildingHitEffect
  (CallFunction|ApplyEffect "/Game/Characters/Enemies/Husk/Materials/M_EnemyHitBuilding_01.M_EnemyHitBuilding_01" (Variables|Effects|GetBuildingHitFadeInTime) (Variables|Effects|GetBuildingHitDuration) (Variables|Effects|GetBuildingHitFadeOutTime)))

(fn ApplyEffect
  (params (SourceMaterial :type Material%20Interface%20Object%20Reference) (FadeInTime :type Float%20%28double-precision%29) (Duration :type Float%20%28double-precision%29) (FadeOutTime :type Float%20%28double-precision%29))
  (bind _effect_mid (LocalVariables|GetEffect%20MID))
  (bind _sourcemeshcomponent (Variables|FortFXSkeletonMeshComponent|GetSourceMeshComponent)
    (:then
      (for _index (range (Rendering|Material|GetNumMaterials _sourcemeshcomponent))
        (bind _returnvalue (Rendering|Material|CreateDynamicMaterialInstance _index SourceMaterial))
        (bind _output_get (LocalVariables|SetEffect%20MID _returnvalue))
        (Rendering|Material|SetScalarParameterValue _output_get "TimestampA" (Utilities|Time|GetGameTimeinSeconds) :!class /Script/Engine.MaterialInstanceDynamic)
        (Rendering|Material|SetScalarParameterValue _output_get "TransitionLength_In" FadeInTime :!class /Script/Engine.MaterialInstanceDynamic)
        (Rendering|Material|SetScalarParameterValue _output_get "TimestampB" (+ (Utilities|Time|GetGameTimeinSeconds) (- Duration FadeOutTime)) :!class /Script/Engine.MaterialInstanceDynamic)
        (Rendering|Material|SetScalarParameterValue _output_get "TransitionLength_Out" FadeOutTime :!class /Script/Engine.MaterialInstanceDynamic)
        (bind _returnvalue_2 (Fortnite|GetDynamicMaterialInstance _sourcemeshcomponent _index))
        (bind _returnvalue_3 (Rendering|Material|GetTextureParameterValue _returnvalue_2 "Normal Map" :!class /Script/Engine.MaterialInstanceDynamic))
        (Utilities|IsValid _returnvalue_3
          (:Is%20Valid
            (Rendering|Material|SetTextureParameterValue (LocalVariables|GetEffect%20MID) "Normal Map" _returnvalue_3))
          (:Is%20Not%20Valid
            _))
        (bind _returnvalue_4 (Rendering|Material|GetTextureParameterValue _returnvalue_2 "Diffuse Map" :!class /Script/Engine.MaterialInstanceDynamic))
        (Utilities|IsValid _returnvalue_4
          (:Is%20Valid
            (Rendering|Material|SetTextureParameterValue _effect_mid "Diffuse" _returnvalue_4))
          (:Is%20Not%20Valid
            _))
        (Rendering|Material|SetMaterial _index _effect_mid)))
    (:else)))
