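; Enemy FX static mesh component that shows timed material effects for awakening (seen through walls) and hitting buildings.
; Both effect functions pass a material and timing variables to ApplyEffect, which creates a dynamic instance for each source mesh material slot.
; It writes start and fade-out timestamps plus transition lengths, leaving fade animation to the material.
; Normal and diffuse textures are copied from the source mesh's dynamic materials.
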(fn ApplyAwakenEffect
  (CallFunction|ApplyEffect "/Game/Characters/Enemies/Husk/Materials/M_EnemyAwokenThroughWalls_01.M_EnemyAwokenThroughWalls_01" (Variables|Effects|GetAwakenFadeInTime) (Variables|Effects|GetAwakenDuration) (Variables|Effects|GetAwakenFadeOutTime)))

(fn ApplyBuildingHitEffect
  (CallFunction|ApplyEffect "/Game/Characters/Enemies/Husk/Materials/M_EnemyHitBuilding_01.M_EnemyHitBuilding_01" (Variables|Effects|GetBuildingHitFadeInTime) (Variables|Effects|GetBuildingHitDuration) (Variables|Effects|GetBuildingHitFadeOutTime)))

(fn ApplyEffect
  (params (Source%20Material :type Material%20Interface%20Object%20Reference) (FadeInTime :type Float%20%28double-precision%29) (Duration :type Float%20%28double-precision%29) (FadeOutTimer :type Float%20%28double-precision%29))
  (bind _effectmid (LocalVariables|GetEffectMID))
  (bind _sourcemeshcomponent (Variables|FortFXStaticMeshComponent|GetSourceMeshComponent)
    (:then
      (for _index (range (Rendering|Material|GetNumMaterials _sourcemeshcomponent))
        (bind _returnvalue (Rendering|Material|CreateDynamicMaterialInstance _index Source%20Material))
        (bind _output_get (LocalVariables|SetEffectMID _returnvalue))
        (Rendering|Material|SetScalarParameterValue _output_get "TimestampA" (Utilities|Time|GetGameTimeinSeconds) :!class /Script/Engine.MaterialInstanceDynamic)
        (Rendering|Material|SetScalarParameterValue _output_get "TransitionLength_In" FadeInTime :!class /Script/Engine.MaterialInstanceDynamic)
        (Rendering|Material|SetScalarParameterValue _output_get "TimestampB" (+ (Utilities|Time|GetGameTimeinSeconds) (- Duration FadeOutTimer)) :!class /Script/Engine.MaterialInstanceDynamic)
        (Rendering|Material|SetScalarParameterValue _output_get "TransitionLength_Out" FadeOutTimer :!class /Script/Engine.MaterialInstanceDynamic)
        (bind _asmaterial_instance_dynamic (Utilities|Casting|CastToMaterialInstanceDynamic (Utilities|Array|Get(acopy) (Rendering|Material|GetMaterials _sourcemeshcomponent) _index)))
        (bind _returnvalue_2 (Rendering|Material|GetTextureParameterValue _asmaterial_instance_dynamic "Normal Map" :!class /Script/Engine.MaterialInstanceDynamic))
        (Utilities|IsValid _returnvalue_2
          (:Is%20Valid
            (Rendering|Material|SetTextureParameterValue (LocalVariables|GetEffectMID) "Normal Map" _returnvalue_2))
          (:Is%20Not%20Valid
            _))
        (bind _returnvalue_3 (Rendering|Material|GetTextureParameterValue _asmaterial_instance_dynamic "Diffuse Map" :!class /Script/Engine.MaterialInstanceDynamic))
        (Utilities|IsValid _returnvalue_3
          (:Is%20Valid
            (Rendering|Material|SetTextureParameterValue _effectmid "Diffuse" _returnvalue_3))
          (:Is%20Not%20Valid
            _))
        (Rendering|Material|SetMaterial _index _effectmid)))
    (:else)))
